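# Master skeletal-mesh material for a collectable backpack cosmetic, showing a reactive facial expression that reverts to neutral after a Blueprint-timestamped duration.
# Inverted vertex-color red selects a narrow band of expression geometry through the opacity mask, while green adds to the mask.
# An optional world position offset scales the selected band around an eye center point, with a blink parameter scaling Z.
# Character skin and pre/post FX functions wrap the result.
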Material /BRCosmetics/Accessories/FORT_Backpacks/Backpack_M_MED_Collectable/Materials/M_MED_Collectable_Backpack_Master
graph {
 properties {
  bAutomaticallySetUsageInEditor=false
  bUseMaterialAttributes=true
  bUsedWithMeshDeformer=true
  bUsedWithMorphTargets=true
  bUsedWithSkeletalMesh=true
  pixelDepthOffsetMode=PDOM_Legacy
  refractionMethod=RM_IndexOfRefraction
  shadingModel=MSM_FromMaterialExpression
  subsurfaceProfile=&SS_HeroSkin_02
  usedShadingModels=Default Lit | Subsurface | Clear Coat | Subsurface Profile | Cloth | Eye
 }
 nodes [
  n1 = MaterialFunctionCall() {materialFunction=&BaseFN_CharacterFunction_PreFX}
  n2 = GetMaterialAttributes(None=MaterialFunctionCall(
    PreskinnedPosition=@n1.Pre skinned Local Position
    Gmap=@n1.Pre skinned Local Position
    MaterialAttributes=@n1.Pre skinned Local Position
    AO=@n1.Pre skinned Local Position
   ) {materialFunction=&CharacterSkinFX}.Result) {attributeGetTypes [#1]}
  n3 = OneMinus(None=VertexColor().R)
  n4 = Divide(
   A=Subtract(A=Time(), B=ScalarParameter() {group=Collectable, parameterName=Collection_GameTimeFromBP})
   B=ScalarParameter() {defaultValue=0.5, group=Collectable, parameterName=Collection_ReactiveLength}
  )
  n5 = LinearInterpolate(
   A=ScalarParameter() {defaultValue=0.56, group=Collectable, parameterName=Collectable_Expression}
   B=ScalarParameter() {defaultValue=0.56, parameterName=Collectable_Neutral_Expresssion}
   Alpha=Saturate(None=Floor(None=@n4))
  ) {constA=0.04, constB=0.56, desc=Default Neutral Expression}
  n6 = Subtract(A=@n3, B=@n5)
  n7 = Subtract(A=@n3, B=Add(A=@n5) {constB=0.05})
  n8 = Ceil(None=Subtract(A=Ceil(None=@n6), B=Ceil(None=@n7)))
  n9 = Add(A=VertexColor().G, B=@n8)
  n10 = VectorParameter() {group=Collectable, parameterName=EyeCenterPoint, defaultValue=<rgba 0 -16 5.5 0>}
  n11 = Multiply(
   A=Subtract(A=WorldPosition().XYZ, B=TransformPosition(None=@n10.RGB) {transformType=TRANSFORMPOSSOURCE_World}) {constB=0}
   B=ScalarParameter() {defaultValue=0.75, group=Collectable, parameterName=Collectable_Expression_WPOInflate}
  ) {constB=2}
  n12 = Constant() {r=1}
  n13 = MaterialFunctionCall(X=@n12, Y=@n12, Z=OneMinus(None=ScalarParameter() {group=Collectable, parameterName=Collectable_FE_Blink_Amount})) {materialFunction=&MakeFloat3}
  n14 = Add(A=@n2.World Position Offset, B=Multiply(A=@n8, B=Multiply(A=@n11, B=@n13.Result)))
  n15 = MaterialFunctionCall(
   AO=@n1.AO
   Distance Blend=@n1.AO
   Fresnel=@n1.AO
   Gmap=@n1.AO
   Pre Skinned Local Position=@n1.AO
   Material=SetMaterialAttributes!(
    MaterialAttributes=@n2.MaterialAttributes
    Opacity Mask=Saturate(None=@n9)
    World Position Offset=StaticSwitchParameter(True=@n14, False=@n2.World Position Offset) {group=Collectable, parameterName=Collectable_Use_WPOInflate}
   ) {attributeSetTypes [#2, #1]}
  ) {materialFunction=&BaseFN_CharacterFunction_PostFX}
  MaterialFunctionCall(In=@n15.Result) {materialFunction=&CINE_MF_CharPostFX}
 ]
}
