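# Physics asset for the M_MED_Collectable backpack cosmetic, defining the capsule bodies and joints for its dangling arm, leg and tongue bones.
# Five chains (two arms, two legs, one tongue) each attach through a small MSTR body to the central pack body, which has two capsules.
# Arms are nearly rigid, legs swing moderately with twist limited to zero, and the tongue gains swing range toward its tip with slight twist.
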PhysicsAsset /BRCosmetics/Accessories/FORT_Backpacks/Backpack_M_MED_Collectable/Meshes/M_MED_Collectable_Pack_Physics
bodies {
 dyn_l_arm_01 [{
   shapeType=Capsule
   length=1.4182
   radius=1.4037
   shapeName=dyn_l_arm_01_capsule
   center=<1.3654 0.0707 -0.1291>
   rotation=<rot 0 -5.8141 -57.3259>
  }]
 dyn_l_arm_02 [{
   shapeType=Capsule
   length=1.3847
   radius=1.3615
   shapeName=dyn_l_arm_02_capsule
   center=<0.6121 0.1104 -0.0231>
   rotation=<rot 0 -3.052 -174.7626>
  }]
 dyn_l_arm_03 [{
   shapeType=Capsule
   length=1.3526
   radius=1.3302
   shapeName=dyn_l_arm_03_capsule
   center=<0.8987 0.0961 -0.0903>
   rotation=<rot 0 10.8596 -76.7488>
  }]
 dyn_l_arm_04 [{
   shapeType=Capsule
   length=4.5119
   radius=3.3468
   shapeName=dyn_l_arm_04_capsule
   center=<4.173 0.2384 1.5112>
   rotation=<rot 0 -79.576 -84.0162>
  }]
 dyn_l_arm_MSTR [{shapeType=Capsule, length=0.505, radius=0.505, shapeName=dyn_l_arm_MSTR_capsule, rotation=<rot 0 -90 0>}]
 dyn_l_leg_01 [{shapeType=Capsule, length=1.5397, radius=3.296, shapeName=dyn_l_leg_01_capsule, center=<4.4454 0 -0.5777>, rotation=<rot -82.5954 -180 180>}]
 dyn_l_leg_02 [{
   shapeType=Capsule
   length=4.1449
   radius=5.0225
   shapeName=dyn_l_leg_02_capsule
   center=<7.441 0.4467 1.3051>
   rotation=<rot 0 94.1219 -76.8774>
  }]
 dyn_l_leg_MSTR [{shapeType=Capsule, length=0.505, radius=0.505, shapeName=dyn_l_leg_MSTR_capsule, rotation=<rot 0 -90 0>}]
 dyn_r_arm_01 [{
   shapeType=Capsule
   length=1.3774
   radius=1.365
   shapeName=dyn_r_arm_01_capsule
   center=<1.0922 -0.105 -0.0954>
   rotation=<rot 0 -3.2417 -169.2684>
  }]
 dyn_r_arm_02 [{
   shapeType=Capsule
   length=1.3738
   radius=1.3396
   shapeName=dyn_r_arm_02_capsule
   center=<0.6861 0.0325 -0.1682>
   rotation=<rot 7.5235 -78.2991 -90>
  }]
 dyn_r_arm_03 [{
   shapeType=Capsule
   length=6.0164
   radius=3.6362
   shapeName=dyn_r_arm_03_capsule
   center=<5.0974 1.7175 1.0427>
   rotation=<rot 0 105.4781 -79.3097>
  }]
 dyn_r_arm_MSTR [{shapeType=Capsule, length=0.505, radius=0.505, shapeName=dyn_r_arm_MSTR_capsule, rotation=<rot 0 -90 0>}]
 dyn_r_leg_01 [{shapeType=Capsule, length=1.5397, radius=1.1887, shapeName=dyn_l_leg_01_capsule, center=<2.4502 0 -0.3184>, rotation=<rot -82.5954 -180 180>}]
 dyn_r_leg_02 [{
   shapeType=Capsule
   length=4.1449
   radius=3.2095
   shapeName=dyn_l_leg_02_capsule
   center=<6.219 0.3587 1.0195>
   rotation=<rot 0 94.1219 -76.8774>
  }]
 dyn_r_leg_MSTR [{shapeType=Capsule, length=0.505, radius=0.505, shapeName=dyn_r_leg_MSTR_capsule, rotation=<rot 0 -90 0>}]
 dyn_r_tongue_01 [{shapeType=Capsule, length=3.2, radius=0.3, shapeName=dyn_r_tongue_01_capsule, center=<0.9036 0 0.15>, rotation=<rot 0 -172.9988 -87.6611>}]
 dyn_r_tongue_02 [{shapeType=Capsule, length=3.5, radius=0.6, shapeName=dyn_r_tongue_02_capsule, center=<0.8 0 0.24>, rotation=<rot 0 2.1266 -89.9095>}]
 dyn_r_tongue_03 [{shapeType=Capsule, length=3.7, radius=0.6, shapeName=dyn_r_tongue_03_capsule, center=<0.4 0 0.25>, rotation=<rot 0 -0.9777 -89.9475>}]
 dyn_r_tongue_04 [{shapeType=Capsule, length=4.1, radius=0.8, shapeName=dyn_r_tongue_04_capsule, center=<1 0 0.11>, rotation=<rot 0 0.4438 -89.3807>}]
 dyn_r_tongue_MSTR [{shapeType=Capsule, length=0.505, radius=0.505, shapeName=dyn_r_tongue_MSTR_capsule, rotation=<rot 0 -90 0>}]
 pack [
  {shapeType=Capsule, length=0.1154, radius=9.0973, shapeName=pack_capsule, center=<-1.0304 -7.9552 -0.1929>, rotation=<rot 3.1912 -90 7.0785>}
  {shapeType=Capsule, length=0, radius=7, shapeName=pack_capsule, center=<0 -8.2798 8.3716>}
 ]
}
constraints [
 {bone1Name=dyn_l_leg_01, bone2Name=dyn_l_leg_MSTR, swing1LimitDegrees=20, swing2LimitDegrees=15, twistLimitDegrees=0}
 {bone1Name=dyn_l_leg_02, bone2Name=dyn_l_leg_01, swing1LimitDegrees=20, swing2LimitDegrees=45, twistLimitDegrees=0}
 {bone1Name=dyn_l_arm_01, bone2Name=dyn_l_arm_MSTR, swing1LimitDegrees=2, swing2LimitDegrees=2, twistLimitDegrees=0}
 {bone1Name=dyn_l_arm_02, bone2Name=dyn_l_arm_01, swing1LimitDegrees=2, swing2LimitDegrees=2, twistLimitDegrees=0}
 {bone1Name=dyn_l_arm_03, bone2Name=dyn_l_arm_02, swing1LimitDegrees=2, swing2LimitDegrees=2, twistLimitDegrees=0}
 {bone1Name=dyn_l_arm_04, bone2Name=dyn_l_arm_03, swing1LimitDegrees=2, swing2LimitDegrees=2, twistLimitDegrees=0}
 {bone1Name=dyn_l_leg_MSTR, bone2Name=pack, swing1LimitDegrees=20, swing2LimitDegrees=15, twistLimitDegrees=0}
 {bone1Name=dyn_l_arm_MSTR, bone2Name=pack, swing1LimitDegrees=2, swing2LimitDegrees=2, twistLimitDegrees=0}
 {bone1Name=dyn_r_arm_MSTR, bone2Name=pack, swing1LimitDegrees=5, swing2LimitDegrees=5, twistLimitDegrees=0}
 {bone1Name=dyn_r_arm_01, bone2Name=dyn_r_arm_MSTR, swing1LimitDegrees=5, swing2LimitDegrees=5, twistLimitDegrees=0}
 {bone1Name=dyn_r_arm_02, bone2Name=dyn_r_arm_01, swing1LimitDegrees=5, swing2LimitDegrees=5, twistLimitDegrees=0}
 {bone1Name=dyn_r_arm_03, bone2Name=dyn_r_arm_02, swing1LimitDegrees=5, swing2LimitDegrees=5, twistLimitDegrees=0}
 {bone1Name=dyn_r_leg_MSTR, bone2Name=pack, swing1LimitDegrees=20, swing2LimitDegrees=15, twistLimitDegrees=0}
 {bone1Name=dyn_r_leg_01, bone2Name=dyn_r_leg_MSTR, swing1LimitDegrees=20, swing2LimitDegrees=15, twistLimitDegrees=0}
 {bone1Name=dyn_r_leg_02, bone2Name=dyn_r_leg_01, swing1LimitDegrees=20, swing2LimitDegrees=45, twistLimitDegrees=0}
 {bone1Name=dyn_r_tongue_MSTR, bone2Name=pack, swing1LimitDegrees=30, swing2LimitDegrees=30, twistLimitDegrees=5}
 {bone1Name=dyn_r_tongue_01, bone2Name=dyn_r_tongue_MSTR, swing1LimitDegrees=30, swing2LimitDegrees=65, twistLimitDegrees=5}
 {bone1Name=dyn_r_tongue_02, bone2Name=dyn_r_tongue_01, swing1LimitDegrees=30, swing2LimitDegrees=90, twistLimitDegrees=5}
 {bone1Name=dyn_r_tongue_03, bone2Name=dyn_r_tongue_02, swing1LimitDegrees=30, swing2LimitDegrees=145, twistLimitDegrees=5}
 {bone1Name=dyn_r_tongue_04, bone2Name=dyn_r_tongue_03, swing1LimitDegrees=30, swing2LimitDegrees=160, twistLimitDegrees=5}
]
